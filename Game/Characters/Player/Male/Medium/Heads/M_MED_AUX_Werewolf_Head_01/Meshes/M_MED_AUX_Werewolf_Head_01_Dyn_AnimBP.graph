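# Post-style dynamics AnimBlueprint for the male Werewolf head, adding secondary motion to hair and earring bones.
# The input pose is cached, then blended by bIsTargeting with a copy whose head bone gets an additive -30 yaw.
# The result runs through chained, spring-constrained Anim Dynamics nodes in actor space, limited to LOD 1 and below.
# Per-bone alphas tune each effect, with dyn_hair_09 at zero.
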AnimBlueprint /Game/Characters/Player/Male/Medium/Heads/M_MED_AUX_Werewolf_Head_01/Meshes/M_MED_AUX_Werewolf_Head_01_Dyn_AnimBP
target_skeleton=/Game/Characters/Player/Common/Fortnite_Base_Head/Fortnite_Base_Head_Export_Skeleton
graphs [
 AnimGraph {nodes [
   n1 = "Transform (Modify) Bone - Bone: head"!(ComponentPose=Local To Component(LocalPose=Use cached pose 'SGI Cache'()), Alpha=1, Rotation="0,-30,0") -> (Pose) {rotationMode=BMM_Additive, scaleMode=BMM_Additive, alphaBoolBlend {blendOption=Linear}, boneToModify {boneName=head}, rotation {yaw=-30}}
   n2 = Component To Local(ComponentPose=@n1)
   n3 = Get bIsTargeting()
   n4 = Use cached pose 'SGI Cache'()
   n5 = Anim Dynamics - Bone: dyn_hair_01(
    ComponentPose=Local To Component(LocalPose=Blend Poses by bool(BlendPose_0=@n2, BlendPose_1=@n4, BlendTime_0=0.15, bActiveValue=@n3) {blendType=Linear})
    Alpha=1
   ) {
    angularDampingOverride=0.98
    angularSpringConstant=90
    bAngularSpring=true
    bGravityOverrideInSimSpace=true
    bLinearSpring=true
    bOverrideLinearDamping=true
    gravityScale=0
    lODThreshold=1
    linearDampingOverride=0.98
    linearSpringConstant=100
    simulationSpace=Actor
    alphaBoolBlend {blendOption=Linear}
    alphaScaleBiasClamp {
     bClampResult=true
     bInterpResult=true
     clampMax=0.25
     clampMin=0.05
     interpSpeedDecreasing=0.5
     interpSpeedIncreasing=0.5
     outRange {max=0.1}
    }
    boundBone {boneName=dyn_hair_01}
    physicsBodyDefinitions [{
      sphereCollisionRadius=0
      boundBone {boneName=dyn_hair_01}
      boxExtents=<30 30 30>
      constraintSetup {
       angularTargetAxis=AxisY
       twistAxis=AxisZ
       angularLimitsMax=<30 2 0>
       angularLimitsMin=<-30 -2 0>
       angularTarget=<0 0 0>
       linearAxesMax=<0 0 5>
       linearAxesMin=<0 0 -1>
      }
      localJointOffset=<0 0 0>
     }]
   }
   n6 = Anim Dynamics - Bone: dyn_hair_02(ComponentPose=@n5, Alpha=0.125) {
    angularDampingOverride=0.98
    angularSpringConstant=90
    bAngularSpring=true
    bGravityOverrideInSimSpace=true
    bLinearSpring=true
    bOverrideAngularDamping=true
    bOverrideLinearDamping=true
    gravityScale=0.125
    lODThreshold=1
    linearDampingOverride=0.98
    linearSpringConstant=100
    simulationSpace=Actor
    alphaBoolBlend {blendOption=Linear}
    alphaScaleBiasClamp {
     bClampResult=true
     bInterpResult=true
     clampMax=0.25
     clampMin=0.05
     interpSpeedDecreasing=0.5
     interpSpeedIncreasing=0.5
     outRange {max=0.1}
    }
    boundBone {boneName=dyn_hair_02}
    physicsBodyDefinitions [{
      sphereCollisionRadius=0
      boundBone {boneName=dyn_hair_02}
      boxExtents=<30 30 30>
      constraintSetup {
       angularTargetAxis=AxisY
       twistAxis=AxisZ
       angularLimitsMax=<2.5 30 0>
       angularLimitsMin=<-2.5 -30 0>
       angularTarget=<0 0 0>
       linearAxesMax=<10 2.5 5>
       linearAxesMin=<-10 -2.5 0>
      }
      localJointOffset=<0 0 -5>
     }]
   }
   n7 = Anim Dynamics - Bone: dyn_hair_03(ComponentPose=@n6, Alpha=0.75, ExternalForce="0,0,0") {
    angularDampingOverride=0.98
    angularSpringConstant=90
    bAngularSpring=true
    bGravityOverrideInSimSpace=true
    bLinearSpring=true
    bOverrideLinearDamping=true
    gravityScale=0.1
    lODThreshold=1
    linearDampingOverride=0.98
    linearSpringConstant=100
    simulationSpace=Actor
    alphaBoolBlend {blendOption=Linear}
    alphaScaleBiasClamp {
     bClampResult=true
     bInterpResult=true
     clampMax=0.25
     clampMin=0.05
     interpSpeedDecreasing=0.5
     interpSpeedIncreasing=0.5
     outRange {max=0.1}
    }
    boundBone {boneName=dyn_hair_03}
    physicsBodyDefinitions [{
      sphereCollisionRadius=0
      boundBone {boneName=dyn_hair_03}
      boxExtents=<30 30 30>
      constraintSetup {
       angularTargetAxis=AxisY
       twistAxis=AxisZ
       angularLimitsMax=<30 2 0>
       angularLimitsMin=<-30 -2 0>
       angularTarget=<0 0 0>
       linearAxesMax=<0 0 5>
       linearAxesMin=<0 0 -1>
      }
      localJointOffset=<0 0 0>
     }]
   }
   n8 = Anim Dynamics - Bone: dyn_hair_04(ComponentPose=@n7, Alpha=0.3, ExternalForce="0,0,0") {
    angularDampingOverride=0.98
    angularSpringConstant=90
    bAngularSpring=true
    bGravityOverrideInSimSpace=true
    bLinearSpring=true
    bOverrideAngularDamping=true
    bOverrideLinearDamping=true
    gravityScale=0.2
    lODThreshold=1
    linearDampingOverride=0.98
    linearSpringConstant=100
    simulationSpace=Actor
    alphaBoolBlend {blendOption=Linear}
    alphaScaleBiasClamp {
     bClampResult=true
     bInterpResult=true
     clampMax=0.25
     clampMin=0.05
     interpSpeedDecreasing=0.5
     interpSpeedIncreasing=0.5
     outRange {max=0.1}
    }
    boundBone {boneName=dyn_hair_04}
    physicsBodyDefinitions [{
      sphereCollisionRadius=0
      boundBone {boneName=dyn_hair_04}
      boxExtents=<30 30 30>
      constraintSetup {
       angularTargetAxis=AxisY
       twistAxis=AxisZ
       angularLimitsMax=<2.5 30 0>
       angularLimitsMin=<-2.5 -30 0>
       angularTarget=<0 0 0>
       linearAxesMax=<10 2.5 5>
       linearAxesMin=<-10 -2.5 0>
      }
      localJointOffset=<0 0 -5>
     }]
   }
   n9 = Anim Dynamics - Bone: dyn_hair_05(ComponentPose=@n8, Alpha=0.85, ExternalForce="0,0,0") {
    angularDampingOverride=0.98
    angularSpringConstant=90
    bAngularSpring=true
    bGravityOverrideInSimSpace=true
    bLinearSpring=true
    bOverrideLinearDamping=true
    gravityScale=0.05
    lODThreshold=1
    linearDampingOverride=0.98
    linearSpringConstant=100
    simulationSpace=Actor
    alphaBoolBlend {blendOption=Linear}
    alphaScaleBiasClamp {
     bClampResult=true
     bInterpResult=true
     clampMax=0.25
     clampMin=0.05
     interpSpeedDecreasing=0.5
     interpSpeedIncreasing=0.5
     outRange {max=0.1}
    }
    boundBone {boneName=dyn_hair_05}
    physicsBodyDefinitions [{
      sphereCollisionRadius=0
      boundBone {boneName=dyn_hair_05}
      boxExtents=<30 30 30>
      constraintSetup {
       angularTargetAxis=AxisY
       twistAxis=AxisZ
       angularLimitsMax=<30 2 0>
       angularLimitsMin=<-30 -2 0>
       angularTarget=<0 0 0>
       linearAxesMax=<2 2 5>
       linearAxesMin=<-2 -2 -2>
      }
      localJointOffset=<0 0 0>
     }]
   }
   n10 = Anim Dynamics - Bone: dyn_hair_06(ComponentPose=@n9, Alpha=0.25, ExternalForce="0,0,0") {
    angularDampingOverride=0.98
    angularSpringConstant=90
    bAngularSpring=true
    bGravityOverrideInSimSpace=true
    bLinearSpring=true
    bOverrideAngularDamping=true
    bOverrideLinearDamping=true
    gravityScale=0.175
    lODThreshold=1
    linearDampingOverride=0.98
    linearSpringConstant=100
    simulationSpace=Actor
    alphaBoolBlend {blendOption=Linear}
    alphaScaleBiasClamp {
     bClampResult=true
     bInterpResult=true
     clampMax=0.25
     clampMin=0.05
     interpSpeedDecreasing=0.5
     interpSpeedIncreasing=0.5
     outRange {max=0.1}
    }
    boundBone {boneName=dyn_hair_06}
    physicsBodyDefinitions [{
      sphereCollisionRadius=0
      boundBone {boneName=dyn_hair_06}
      boxExtents=<30 30 30>
      constraintSetup {
       angularTargetAxis=AxisY
       twistAxis=AxisZ
       angularLimitsMax=<2.5 30 0>
       angularLimitsMin=<-2.5 -30 0>
       angularTarget=<0 0 0>
       linearAxesMax=<10 2.5 5>
       linearAxesMin=<-10 -2.5 0>
      }
      localJointOffset=<0 0 -5>
     }]
   }
   n11 = Anim Dynamics - Bone: dyn_hair_07(ComponentPose=@n10, Alpha=0.1, ExternalForce="0,0,0") {
    angularDampingOverride=0.98
    angularSpringConstant=90
    bAngularSpring=true
    bGravityOverrideInSimSpace=true
    bLinearSpring=true
    bOverrideLinearDamping=true
    gravityScale=0
    lODThreshold=1
    linearDampingOverride=0.98
    linearSpringConstant=100
    simulationSpace=Actor
    alphaBoolBlend {blendOption=Linear}
    alphaScaleBiasClamp {
     bClampResult=true
     bInterpResult=true
     clampMax=0.25
     clampMin=0.05
     interpSpeedDecreasing=0.5
     interpSpeedIncreasing=0.5
     outRange {max=0.1}
    }
    boundBone {boneName=dyn_hair_07}
    physicsBodyDefinitions [{
      sphereCollisionRadius=0
      boundBone {boneName=dyn_hair_07}
      boxExtents=<30 30 30>
      constraintSetup {
       angularTargetAxis=AxisY
       twistAxis=AxisZ
       angularLimitsMax=<30 2 0>
       angularLimitsMin=<-30 -2 0>
       angularTarget=<0 0 0>
       linearAxesMax=<0 0 5>
       linearAxesMin=<0 0 -1>
      }
      localJointOffset=<0 0 0>
     }]
   }
   n12 = Anim Dynamics - Bone: dyn_hair_08(ComponentPose=@n11, Alpha=0.2, ExternalForce="0,0,0") {
    angularDampingOverride=0.98
    angularSpringConstant=80
    bAngularSpring=true
    bGravityOverrideInSimSpace=true
    bLinearSpring=true
    bOverrideAngularDamping=true
    bOverrideLinearDamping=true
    gravityScale=0.1
    lODThreshold=1
    linearDampingOverride=0.98
    linearSpringConstant=90
    simulationSpace=Actor
    alphaBoolBlend {blendOption=Linear}
    alphaScaleBiasClamp {
     bClampResult=true
     bInterpResult=true
     clampMax=0.25
     clampMin=0.05
     interpSpeedDecreasing=0.5
     interpSpeedIncreasing=0.5
     outRange {max=0.1}
    }
    boundBone {boneName=dyn_hair_08}
    physicsBodyDefinitions [{
      sphereCollisionRadius=0
      boundBone {boneName=dyn_hair_08}
      boxExtents=<30 30 30>
      constraintSetup {
       angularTargetAxis=AxisY
       twistAxis=AxisZ
       angularLimitsMax=<2.5 30 0>
       angularLimitsMin=<-2.5 -30 0>
       angularTarget=<0 0 0>
       linearAxesMax=<1 1 2>
       linearAxesMin=<-1 -1 0>
      }
      localJointOffset=<0 0 0>
     }]
   }
   n13 = Anim Dynamics - Bone: dyn_hair_09(ComponentPose=@n12, Alpha=0, ExternalForce="0,0,0") {
    angularDampingOverride=0.98
    angularSpringConstant=90
    bAngularSpring=true
    bGravityOverrideInSimSpace=true
    bLinearSpring=true
    bOverrideLinearDamping=true
    gravityScale=0
    lODThreshold=1
    linearDampingOverride=0.98
    linearSpringConstant=100
    simulationSpace=Actor
    alphaBoolBlend {blendOption=Linear}
    alphaScaleBiasClamp {
     bClampResult=true
     bInterpResult=true
     clampMax=0.25
     clampMin=0.05
     interpSpeedDecreasing=0.5
     interpSpeedIncreasing=0.5
     outRange {max=0.1}
    }
    boundBone {boneName=dyn_hair_09}
    physicsBodyDefinitions [{
      sphereCollisionRadius=0
      boundBone {boneName=dyn_hair_09}
      boxExtents=<30 30 30>
      constraintSetup {
       angularTargetAxis=AxisY
       twistAxis=AxisZ
       angularLimitsMax=<30 2 0>
       angularLimitsMin=<-30 -2 0>
       angularTarget=<0 0 0>
       linearAxesMax=<0 0 1>
      }
      localJointOffset=<0 0 0>
     }]
   }
   n14 = Anim Dynamics - Bone: dyn_earing_r(ComponentPose=@n13, Alpha=0.2, ExternalForce="0,0,0") {
    angularDampingOverride=1
    angularSpringConstant=2000
    bAngularSpring=true
    bGravityOverrideInSimSpace=true
    bLinearSpring=true
    bOverrideAngularDamping=true
    bOverrideLinearDamping=true
    gravityScale=0.1
    lODThreshold=1
    linearDampingOverride=1
    linearSpringConstant=400
    simulationSpace=Actor
    alphaBoolBlend {blendOption=Linear}
    boundBone {boneName=dyn_earing_r}
    physicsBodyDefinitions [{
      sphereCollisionRadius=0
      boundBone {boneName=dyn_earing_r}
      boxExtents=<40 40 40>
      constraintSetup {
       angularTargetAxis=AxisY
       angularLimitsMax=<0 30 30>
       angularLimitsMin=<0 -30 -30>
       angularTarget=<0 0 0>
       linearAxesMax=<10 20 20>
       linearAxesMin=<0 -5 -5>
      }
      localJointOffset=<0 -0.5 0>
     }]
   }
   n15 = Anim Dynamics - Bone: dyn_earing_l(ComponentPose=@n14, Alpha=0.15, ExternalForce="0,0,0") {
    angularDampingOverride=1
    angularSpringConstant=2000
    bAngularSpring=true
    bGravityOverrideInSimSpace=true
    bLinearSpring=true
    bOverrideAngularDamping=true
    bOverrideLinearDamping=true
    gravityScale=0.1
    lODThreshold=1
    linearDampingOverride=1
    linearSpringConstant=400
    simulationSpace=Actor
    alphaBoolBlend {blendOption=Linear}
    boundBone {boneName=dyn_earing_l}
    physicsBodyDefinitions [{
      sphereCollisionRadius=0
      boundBone {boneName=dyn_earing_l}
      boxExtents=<40 40 40>
      constraintSetup {
       angularTargetAxis=AxisY
       angularLimitsMax=<0 30 30>
       angularLimitsMin=<0 -30 -30>
       angularTarget=<0 0 0>
       linearAxesMax=<10 20 5>
       linearAxesMin=<0 0 -20>
      }
      localJointOffset=<0 0 0>
     }]
   }
   Output Pose(Result=Component To Local(ComponentPose=@n15)) {layerGroup=None}
   AnimGraphNode_SubInput_0 = Input Pose()
   Save cached pose 'SGI Cache'(Pose=@AnimGraphNode_SubInput_0) {cacheName=SGI Cache}
  ]}
 EventGraph {nodes [Event BlueprintUpdateAnimation()]}
]
variables []
